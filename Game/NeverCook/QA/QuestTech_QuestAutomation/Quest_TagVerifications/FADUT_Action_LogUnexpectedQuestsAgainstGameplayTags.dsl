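; Editor QA action that flags quest assets with unexpected Creative product tags, judging intent by asset name.
; RunActionOnAssetDataArray calls its parent, then searches each loaded asset's display name for a configurable substring with case and direction settings.
; Names lacking the substring log a warning; matches log as intended Creative only when bLogAllResults is set.
; No tags are read here, so the incoming array is assumed to be tag-filtered already.
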(fn RunActionOnAssetDataArray
  (params (AssetDataArray :type Array%20of%20Asset%20Data%20Structures))
  (bind _returnvalue (|CallParentFunction AssetDataArray))
  (for _array_element AssetDataArray
    (if (== (Utilities|String|FindSubstring (Utilities|GetDisplayName (AssetData|GetAsset _array_element)) (Variables|SearchSettings|GetQuestSubstringToCompareAgainst) (Variables|SearchSettings|GetbIsCaseSensitive) (Variables|SearchSettings|GetbStartSearchFromEndOfString)) -1)
      (Development|PrintString 0 (Utilities|String|Append "WARNING: This Quest has creative Product Tags: " (Utilities|GetDisplayName (AssetData|GetAsset _array_element))) false)
      (else
        (Development|PrintString 0 (Utilities|String|Append "Looks like this quest is intended to be creative: " (Utilities|GetDisplayName (AssetData|GetAsset _array_element))) false (Variables|ReportLogs|GetbLogAllResults)))))
  (return _returnvalue))
